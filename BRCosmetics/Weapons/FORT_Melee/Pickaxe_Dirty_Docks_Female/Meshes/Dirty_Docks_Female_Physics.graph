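# Physics asset for the Dirty Docks Female pickaxe mesh, defining collision bodies and joints for its two dangling rope chains.
# The handle body combines a capsule with four angled boxes around the upper end.
# A three-bone bottom rope and a six-bone top rope hang from the handle through swing- and twist-limited constraints.
# The top rope forks at its third bone, with both the fourth and fifth bones constrained to it.
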PhysicsAsset /BRCosmetics/Weapons/FORT_Melee/Pickaxe_Dirty_Docks_Female/Meshes/Dirty_Docks_Female_Physics
bodies {
 Melee_TwoHanded_Handle [
  {shapeType=Capsule, length=91.2203, radius=4.3398, center=<0 0 13.7769>}
  {shapeType=Box, extentX=42.193, extentY=0.8419, extentZ=61.2669, center=<-15.0129 -8.3146 65.933>, rotation=<rot 33.7527 25.0162 -4.1091>}
  {shapeType=Box, extentX=7.5, extentY=0.5, extentZ=20.8495, center=<15.4933 -8.7796 64.0356>, rotation=<rot -30.748 -30.1408 1.4067>}
  {shapeType=Box, extentX=6.1253, extentY=2.7867, extentZ=19.4803, center=<-5.8314 -3.937 74.578>, rotation=<rot 19.7304 21.321 -6.8815>}
  {shapeType=Box, extentX=6.1253, extentY=2.7867, extentZ=19.4803, center=<6.5875 -3.0835 74.191>, rotation=<rot -22.5008 -31.5413 -3.9479>}
 ]
 dyn_ropeBtm_01 [{shapeType=Capsule, length=1.8163, radius=1.2813, center=<1.8039 0 0>, rotation=<rot 90 0 0>}]
 dyn_ropeBtm_02 [{shapeType=Capsule, length=1.1155, radius=1.1875, center=<1.732 0 0>, rotation=<rot 90 0 0>}]
 dyn_ropeBtm_03 [{shapeType=Capsule, length=1.6191, radius=1, center=<1.6595 0 0>, rotation=<rot 90 0 0>}]
 dyn_ropeTop_01 [{shapeType=Box, extentX=6.809, extentY=7.8501, extentZ=0.7984, center=<3.2796 0 0>}]
 dyn_ropeTop_02 [{shapeType=Box, extentX=3.7435, extentY=3.8225, extentZ=1.3115, center=<3.2796 0 0>}]
 dyn_ropeTop_03 [{shapeType=Capsule, length=2.0514, radius=0.9872, center=<1.5107 0 0>, rotation=<rot 90 0 0>}]
 dyn_ropeTop_04 [{shapeType=Capsule, length=1.3865, radius=0.8688, center=<1.701 0 0.0001>, rotation=<rot 90 0 0>}]
 dyn_ropeTop_05 [{shapeType=Capsule, length=1.5254, radius=0.9081, center=<1.8668 0 -0.3473>, rotation=<rot 90 0 0>}]
 dyn_ropeTop_06 [{shapeType=Capsule, length=0.8626, radius=1.3232, center=<1.6693 0 -0.2312>, rotation=<rot 90 0 0>}]
}
constraints [
 {bone1Name=dyn_ropeBtm_03, bone2Name=dyn_ropeBtm_02, swing1LimitDegrees=45, swing2LimitDegrees=45, twistLimitDegrees=10}
 {bone1Name=dyn_ropeBtm_01, bone2Name=Melee_TwoHanded_Handle, swing1LimitDegrees=45, swing2LimitDegrees=45, twistLimitDegrees=10}
 {bone1Name=dyn_ropeBtm_02, bone2Name=dyn_ropeBtm_01, swing1LimitDegrees=45, swing2LimitDegrees=45, twistLimitDegrees=10}
 {bone1Name=dyn_ropeTop_01, bone2Name=Melee_TwoHanded_Handle, swing1LimitDegrees=60, swing2LimitDegrees=60, twistLimitDegrees=15}
 {bone1Name=dyn_ropeTop_02, bone2Name=dyn_ropeTop_01, swing1LimitDegrees=20, swing2LimitDegrees=30, twistLimitDegrees=15}
 {bone1Name=dyn_ropeTop_03, bone2Name=dyn_ropeTop_02, swing1LimitDegrees=20, swing2LimitDegrees=30, twistLimitDegrees=45}
 {bone1Name=dyn_ropeTop_04, bone2Name=dyn_ropeTop_03, swing1LimitDegrees=20, swing2LimitDegrees=30, twistLimitDegrees=45}
 {bone1Name=dyn_ropeTop_05, bone2Name=dyn_ropeTop_03, swing1LimitDegrees=20, swing2LimitDegrees=30, twistLimitDegrees=45}
 {bone1Name=dyn_ropeTop_06, bone2Name=dyn_ropeTop_05, swing1LimitDegrees=20, swing2LimitDegrees=30, twistLimitDegrees=45}
]
